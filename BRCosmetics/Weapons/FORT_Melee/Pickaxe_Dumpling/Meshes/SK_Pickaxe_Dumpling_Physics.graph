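# Physics asset for the Dumpling two-handed pickaxe skeletal mesh, defining collision bodies and joints for its dangling rings and ropes.
# A multi-capsule body on Melee_TwoHanded_Handle anchors three chains: left and right ring-plus-three-rope chains, and a three-segment ring chain.
# All joints lock twist; chain roots and the middle ring joint swing on one axis as limited hinges, while rope joints and the last ring swing on two.
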PhysicsAsset /BRCosmetics/Weapons/FORT_Melee/Pickaxe_Dumpling/Meshes/SK_Pickaxe_Dumpling_Physics
bodies {
 Melee_TwoHanded_Handle [
  {shapeType=Capsule, length=66.0709, radius=6.9847, center=<-0.3186 0.9927 24.1646>, rotation=<rot 3.5532 -89.9996 0.0609>}
  {shapeType=Capsule, length=19.2542, radius=6.7723, center=<7.0538 6.4318 87.2154>, rotation=<rot 9.4709 -3.3829 -19.7593>}
  {shapeType=Capsule, length=23.5652, radius=6.7723, center=<-7.6421 7.8082 87.5072>, rotation=<rot -6.3192 2.2449 -19.603>}
  {shapeType=Capsule, length=23.4972, radius=8.18, center=<-0.7973 8.8062 88.4328>, rotation=<rot -0.0286 0.0101 -26.8262>}
  {shapeType=Capsule, length=7.3301, radius=5.004, center=<5.3495 9.2546 67.8842>, rotation=<rot -34.017 -8.3201 18.4802>}
  {shapeType=Capsule, length=7.3301, radius=5.004, center=<-5.3072 8.4155 67.8101>, rotation=<rot 37.6088 14.3797 19.3694>}
 ]
 dyn_ling_l_01 [{shapeType=Box, extentX=7.5526, extentY=1.9413, extentZ=8.4579, center=<-3.1498 0.1519 0.4164>, rotation=<rot -0.7042 -6.8153 1.0766>}]
 dyn_lope_l_01 [{shapeType=Capsule, length=8.3184, radius=1.6389, center=<-4.0185 0.5462 -0.3883>, rotation=<rot 0 83.2543 -85.9783>}]
 dyn_lope_l_02 [{shapeType=Capsule, length=7.7483, radius=1.6647, center=<-4.0728 0.1377 0.1246>, rotation=<rot 0.1484 88.6541 -88.716>}]
 dyn_lope_l_03 [
  {shapeType=Capsule, length=0, radius=5.9489, center=<-6.7242 -0.8295 -0.3036>, rotation=<rot 0 -100.6507 -87.5631>}
  {shapeType=Capsule, length=4.2327, radius=2.1436, center=<0.2522 -1.2927 0.3705>, rotation=<rot 4.4719 -15.8517 -74.6439>}
 ]
 dyn_ring_00 [{shapeType=Capsule, length=5.6341, radius=2.0485, center=<2.5736 -0.0001 0.0316>, rotation=<rot -89.2964 -0.0001 0>}]
 dyn_ring_01 [{shapeType=Capsule, length=0.4481, radius=2.4411, center=<-3.8917 -0.0001 -0.0478>, rotation=<rot -89.2964 -0.0001 0>}]
 dyn_ring_02 [{shapeType=Capsule, length=1.4233, radius=5.3659, center=<6.2383 0.0001 0.0766>, rotation=<rot -89.2964 -0.0001 0>}]
 dyn_ring_r_01 [
  {shapeType=Sphere, radius=1.5, center=<-0.5591 0 -0.731>}
  {shapeType=Box, extentX=3.4111, extentY=1.5326, extentZ=8.2387, center=<2.7118 -0.3013 -0.6434>, rotation=<rot 1.8085 -4.9768 18.6974>}
 ]
 dyn_rope_r_01 [{shapeType=Capsule, length=7.4638, radius=2, center=<4.069 -0.4497 -0.407>, rotation=<rot 0 85.4664 -90.0982>}]
 dyn_rope_r_02 [{shapeType=Capsule, length=7.6587, radius=2, center=<4.1181 -0.0286 -0.2876>, rotation=<rot 0.0233 -91.1347 -89.8253>}]
 dyn_rope_r_03 [
  {shapeType=Capsule, length=0, radius=5.8751, center=<6.4134 0.9104 -0.356>, rotation=<rot 0 80.1637 -89.4777>}
  {shapeType=Capsule, length=4.2327, radius=2.1436, center=<0.3402 1.5402 -0.3052>, rotation=<rot 3.9517 -13.9622 -74.5048>}
 ]
}
constraints [
 {bone1Name=dyn_rope_r_01, bone2Name=dyn_ring_r_01, swing1LimitDegrees=45, swing2LimitDegrees=45, twistLimitDegrees=1, twistMotion=Locked}
 {bone1Name=dyn_rope_r_02, bone2Name=dyn_rope_r_01, swing1LimitDegrees=45, swing2LimitDegrees=45, twistLimitDegrees=1, twistMotion=Locked}
 {bone1Name=dyn_rope_r_03, bone2Name=dyn_rope_r_02, swing1LimitDegrees=45, swing2LimitDegrees=45, twistLimitDegrees=1, twistMotion=Locked}
 {bone1Name=dyn_lope_l_01, bone2Name=dyn_ling_l_01, swing1LimitDegrees=45, swing2LimitDegrees=45, twistLimitDegrees=1, twistMotion=Locked}
 {bone1Name=dyn_lope_l_02, bone2Name=dyn_lope_l_01, swing1LimitDegrees=45, swing2LimitDegrees=45, twistLimitDegrees=1, twistMotion=Locked}
 {bone1Name=dyn_lope_l_03, bone2Name=dyn_lope_l_02, swing1LimitDegrees=45, swing2LimitDegrees=45, twistLimitDegrees=1, twistMotion=Locked}
 {
  bone1Name=dyn_ling_l_01
  bone2Name=Melee_TwoHanded_Handle
  swing1LimitDegrees=40
  swing2LimitDegrees=0
  swing2Motion=Locked
  twistLimitDegrees=0
  twistMotion=Locked
 }
 {
  bone1Name=dyn_ring_r_01
  bone2Name=Melee_TwoHanded_Handle
  swing1LimitDegrees=40
  swing2LimitDegrees=0
  swing2Motion=Locked
  twistLimitDegrees=0
  twistMotion=Locked
 }
 {
  bone1Name=dyn_ring_00
  bone2Name=Melee_TwoHanded_Handle
  swing1LimitDegrees=60
  swing2LimitDegrees=1
  swing2Motion=Locked
  twistLimitDegrees=0.01
  twistMotion=Locked
 }
 {
  bone1Name=dyn_ring_01
  bone2Name=dyn_ring_00
  swing1LimitDegrees=5
  swing2LimitDegrees=0
  swing2Motion=Locked
  twistLimitDegrees=0.01
  twistMotion=Locked
 }
 {bone1Name=dyn_ring_02, bone2Name=dyn_ring_01, swing1LimitDegrees=60, swing2LimitDegrees=60, twistLimitDegrees=0, twistMotion=Locked}
]
